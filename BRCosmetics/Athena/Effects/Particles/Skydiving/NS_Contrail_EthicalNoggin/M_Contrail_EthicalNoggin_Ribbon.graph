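# Translucent, unlit, two-sided material for the EthicalNoggin skydiving contrail ribbon in Niagara.
# Dynamic parameters offset, speed and reverse animated sine waves that, with swirl-normal warping, distort UVs for three offset fire-noise samples, creating chromatic fringing.
# A rotated gradient carves hot-edge and caustic-filled center masks for time-of-day-adjusted emissive and end-faded opacity, with cheaper mobile branches.
# A separate branch computes a Z offset at one end for near-vertical camera angles.
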Material /BRCosmetics/Athena/Effects/Particles/Skydiving/NS_Contrail_EthicalNoggin/M_Contrail_EthicalNoggin_Ribbon
graph {
 properties {
  bAutomaticallySetUsageInEditor=false
  bUsedWithNiagaraSprites=true
  bUsedWithStaticMesh=false
  blendMode=BLEND_Translucent
  pixelDepthOffsetMode=PDOM_Legacy
  shadingModel=MSM_Unlit
  twoSided=true
  usedShadingModels=Unlit
 }
 nodes [
  n1 = ScalarParameter() {defaultValue=0.025, parameterName=ChromaShiftDistance}
  n2 = Multiply(A=Constant2Vector() {r=1}, B=@n1)
  n3 = TextureCoordinate()
  n4 = AppendVector(A=ScalarParameter() {defaultValue=2, parameterName=WarpTexTilingU}, B=ScalarParameter() {defaultValue=1, parameterName=WarpTexTilingV})
  n5 = ScalarParameter() {parameterName=WarpPanSpeedU}
  n6 = ScalarParameter() {defaultValue=0.25, parameterName=WarpPanSpeedV}
  n7 = TextureSampleParameter2D(UVs=Panner(Coordinate=Multiply(A=@n3, B=@n4), Speed=AppendVector(A=@n5, B=@n6)) {bFractionalPart=true, speedY=0.25}) {parameterName=UVDistTex, samplerType=SAMPLERTYPE_Normal, texture=/Game/Global/Textures/Default/Noise/T_SwirlNoise_N}
  n8 = ScalarParameter() {defaultValue=2, parameterName=BaseTilingU}
  n9 = ScalarParameter() {defaultValue=1, parameterName=BaseTilingV}
  n10 = DynamicParameter() {defaultValue=<rgba 1 2 1 1>, paramNames ["offset(0-1)", speed, "reverse(0,1)", "Rand(0-1)"]}
  n11 = TextureCoordinate()
  n12 = ComponentMask(None=@n11) {r=true}
  n13 = AppendVector(A=Abs(None=Subtract(A=@n10."reverse(0,1)", B=@n12)), B=ComponentMask(None=@n11) {g=true})
  n14 = ComponentMask(None=@n13) {g=true}
  n15 = Sine(None=Divide(A=OneMinus(None=@n14)) {constB=4})
  n16 = Add(A=@n13, B=AppendVector(A=Constant(), B=@n10."offset(0-1)"))
  n17 = WorldPosition()
  n18 = Constant() {r=-1618}
  n19 = Add(A=Multiply(A=Time(), B=@n10.speed) {constB=10}, B=Divide(A=ComponentMask(None=@n17.XYZ) {b=true}, B=Divide(A=@n18) {constB=4}) {constB=1618})
  n20 = Frac(None=ComponentMask(None=Panner(Coordinate=@n16, Time=@n19) {speedY=0.2}) {g=true})
  n21 = ScalarParameter() {defaultValue=1, group=Distort_map, parameterName=Disort_frequency, sortPriority=4}
  n22 = Panner(Coordinate=Add(A=@n16, B=Constant2Vector() {r=0.2}), Time=@n19) {speedY=0.3}
  n23 = Multiply(A=Frac(None=ComponentMask(None=@n22) {g=true}), B=@n21)
  n24 = Add(A=Sine(None=Multiply(A=@n20, B=@n21)) {period=0.5}, B=Multiply(A=Cosine(None=@n23)) {constB=0.1})
  n25 = Constant2Vector() {g=0.5, r=0.4}
  n26 = TextureSampleParameter2D(UVs=Panner(Coordinate=Multiply(A=@n13, B=@n25), Time=Time()) {speedY=0.2}) {
   group=Distort_map
   parameterName=Detail_disort_map
   samplerType=SAMPLERTYPE_Normal
   texture=/Game/Global/Textures/Default/Tileable_Patterns/1Normal
  }
  n27 = Add(A=@n24, B=@n26.R) {constB=0}
  n28 = ShadingPathSwitch(Default=@n27, Mobile=@n27) {desc=Migrated from FeatureLevelSwitch}
  n29 = Add(
   A=@n13
   B=Multiply(A=Multiply(A=@n15, B=@n28), B=ScalarParameter() {defaultValue=0.2, group=Distort_map, parameterName=Disort_amplitude, sortPriority=5}) {constB=0.15}
  )
  n30 = Add(
   A=Multiply(A=ComponentMask(None=@n7.RGB) {g=true, r=true}, B=ScalarParameter() {defaultValue=0.07, parameterName=UVWarpStr}) {constB=0.0125}
   B=Multiply(A=AppendVector(A=@n8, B=@n9), B=@n29)
  )
  n31 = Multiply(A=@n1, B=Constant2Vector() {r=-1})
  n32 = MaterialFunctionCall(
   X=TextureSampleParameter2D(UVs=Add(A=@n2, B=@n30)) {parameterName=GrayscaleTex, texture=/Game/Global/Textures/Default/Noise/T_FireNoise_02}.G
   Y=TextureSampleParameter2D(UVs=@n30) {parameterName=GrayscaleTex_1, texture=&T_FireNoise_02}.G
   Z=TextureSampleParameter2D(UVs=Add(A=@n30, B=@n31)) {parameterName=GrayscaleTex_2, texture=&T_FireNoise_02}.G
  ) {materialFunction=&MakeFloat3}
  n33 = ScalarParameter() {defaultValue=-1, parameterName=SaturationFactor}
  n34 = ShadingPathSwitch(Default=Clamp(None=Desaturation(None=@n32.Result, Fraction=@n33)), Mobile=Constant() {r=1}) {desc=Migrated from FeatureLevelSwitch}
  n35 = Constant3Vector() {constant=<rgba 0 0.5675 1 1>}
  n36 = TextureSample(UVs=MaterialFunctionCall(UVs=@n29, "Rotation Angle (0-1)"=Constant() {r=0.25}) {materialFunction=&CustomRotator}.Rotated Values) {texture=/Game/Global/Textures/Default/Masks_and_Gradients/gradient_stasis}
  n37 = Clamp(None=Sine(None=Add(A=@n36.R) {constB=0.9}))
  n38 = Multiply(A=Add(A=@n34, B=Multiply(A=@n35, B=@n37))) {constB=2}
  n39 = MaterialFunctionCall(Alpha=ComponentMask(None=TextureCoordinate()) {g=true}, Min=Constant() {r=0.1}, Max=Constant() {r=0.4}) {materialFunction=&SmoothStep}
  n40 = ScalarParameter() {defaultValue=5, group=Specular, parameterName=Specular_intensity, sortPriority=1}
  n41 = Multiply(A=@n34, B=Multiply(A=Multiply(A=@n34, B=@n35), B=@n40))
  n42 = Multiply(A=@n41) {constB=0.5}
  n43 = ShadingPathSwitch(
   Default=Multiply(A=MaterialFunctionCall(Color_Morning=@n41, Color_Day=@n41, Color_Evening=@n41, Color_Night=@n42) {materialFunction=&Color_Time_of_Day}.Result) {constB=0.1}
   Mobile=@n34
  ) {desc=Migrated from FeatureLevelSwitch}
  n44 = Multiply(A=Multiply(A=Multiply(A=@n38, B=@n39.Result), B=@n37), B=Clamp(None=Add(B=@n43) {constA=0.2}))
  n45 = Add(A=@n36.R) {constB=0.8}
  n46 = MaterialFunctionCall(Alpha=Clamp(None=Sine(None=@n45)), Min=Constant() {r=0.9}, Max=Constant() {r=1}) {materialFunction=&SmoothStep}
  n47 = Multiply(A=@n40, B=Clamp(None=Multiply(A=@n39.Result, B=@n46.Result)) {desc=Hot Edge Mask})
  n48 = ComponentMask(None=TextureCoordinate()) {g=true}
  n49 = MaterialFunctionCall(Alpha=@n36.R, Min=Constant() {r=0.4}, Max=Constant() {r=0.5}) {desc=Center Mask, materialFunction=&SmoothStep}
  n50 = Time()
  n51 = Multiply(
   A=@n49.Result
   B=TextureSample(UVs=Panner(Coordinate=@n29, Time=@n50) {speedY=0.5}) {texture=/Game/Global/Textures/Default/Caustics/T_Caustics_Softer}.RGB
  )
  n52 = Multiply(A=Multiply(A=@n43, B=Multiply(A=@n48, B=@n48)), B=@n51)
  n53 = Constant()
  MaterialFunctionCall(Emissive=Add(A=Add(A=@n44, B=@n47), B=ShadingPathSwitch(Default=@n52, Mobile=@n53) {desc=Migrated from FeatureLevelSwitch})) {materialFunction=&MF_FX_EmissiveAdjustments}
  n54 = TextureCoordinate()
  n55 = OneMinus(None=MaterialFunctionCall(Alpha=ComponentMask(None=@n54) {g=true}, Min=ScalarParameter() {defaultValue=0.8, parameterName=PinchHeight}, Max=Constant() {r=1}) {materialFunction=&SmoothStep}.Result)
  n56 = Multiply(A=@n51, B=Clamp(None=Multiply(A=@n48) {constB=12}))
  n57 = ComponentMask(None=TextureCoordinate()) {g=true}
  n58 = Power(Base=Clamp(None=Multiply(A=@n57) {constB=2})) {constExponent=4}
  n59 = Clamp(None=Multiply(A=OneMinus(None=@n57)) {constB=20})
  n60 = Add(A=Multiply(A=@n37, B=@n39.Result), B=Multiply(A=@n49.Result, B=Multiply(A=@n58, B=@n59)))
  n61 = Add(A=@n56, B=Multiply(A=Clamp(None=@n60), B=Multiply(A=@n59, B=@n59)))
  Multiply(A=@n55, B=Multiply(A=@n55, B=Multiply(A=@n55, B=@n61)))
  n62 = OneMinus(None=ComponentMask(None=TextureCoordinate()) {g=true})
  n63 = Constant() {r=0.9}
  n64 = Constant() {r=1}
  n65 = Constant()
  n66 = ObjectPositionWS()
  n67 = AppendVector(A=Constant2Vector(), B=ComponentMask(None=@n66) {b=true})
  n68 = Subtract(A=CameraPositionWS(), B=@n66)
  n69 = Abs(None=DotProduct(A=Normalize(VectorInput=@n67), B=Normalize(VectorInput=@n68)))
  n70 = Constant() {r=0.9}
  n71 = Constant() {r=30}
  n72 = Constant()
  Multiply(
   A=Multiply(A=If(A=@n62, B=@n63, "A > B"=@n64, "A == B"=@n65, "A < B"=@n65))
   B=AppendVector(A=Constant2Vector(), B=If(A=@n69, B=@n70, "A > B"=@n71, "A == B"=@n72, "A < B"=@n72))
  )
  Reroute(None=@n24)
  Reroute(None=@n24)
  n73 = Constant()
  Reroute(None=@n73)
  Reroute(None=@n53)
  ShadingPathSwitch(Default=@n56, Mobile=@n73) {desc=Migrated from FeatureLevelSwitch}
 ]
}
