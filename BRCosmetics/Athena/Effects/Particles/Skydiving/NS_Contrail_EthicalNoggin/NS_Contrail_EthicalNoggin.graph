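# Niagara skydiving contrail for the EthicalNoggin cosmetic, attaching effects to four sockets through a shared socket reader.
# Each emitter partitions particles by socket index, with some sizes and parameters varying per socket.
# Glow sprites pulse with smoothed audio, "Ribbon" draws aligned sprites plus lights rather than ribbons, and faint decal-line sprites follow the sockets.
# GPU sparks blend launch velocity by dive amount, with separate gameplay and front-end values.
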NiagaraSystem /BRCosmetics/Athena/Effects/Particles/Skydiving/NS_Contrail_EthicalNoggin/NS_Contrail_EthicalNoggin
dependencies {
 usedDataInterfaces [
  &NiagaraDataInterfaceArrayFloat2
  &NiagaraDataInterfaceCurve
  &NiagaraDataInterfaceVector2DCurve
  &NiagaraDataInterfaceArrayFloat
  &NiagaraDataInterfaceSocketReader
 ]
 usedDynamicInputs [
  /Niagara/DynamicInputs/Integer/V2/Round_Float
  &Multiply_Vector
  &RotateVectorByQuaternion
  /Niagara/DynamicInputs/Rotation/InvertQuaternion
  &TransformVector
  &SetBoolByFloatComparison
 ]
 usedModules [
  &self:SystemSpawnScript.SystemScriptSource.SystemScriptGraph.NiagaraNodeAssignment_0.SetVariables_#1
  &SystemState
  /Game/Effects/Niagara/Modules/Skydiving/FN_DivingAttributes/v5/FN_DivingAttributes
  &self:SystemSpawnScript.SystemScriptSource.SystemScriptGraph.NiagaraNodeAssignment_1.SetVariables_#2
  &FN_Integrator
 ]
}
emitters {
 Decal_Glow {
  emitterUpdateScript [
   &EmitterState(Life Cycle Mode=::NewEnumerator1, Loop Behavior=::NewEnumerator1, Loop Duration Mode=::NewEnumerator1, UseLoopDelay=false)
   &SpawnBurst_Instantaneous(Spawn Count=4, Use Spawn Probability=false, UseLoopCountLimit=false)
  ]
  particleSpawnScript [
   &InitializeParticle(
    Color=<rgba 0.75 1.6838 3 1>
    Lifetime=&RandomRangeFloat(Minimum=1, Maximum=2)
    Sprite Rotation Mode=::NewEnumerator2
    Sprite Size Mode=::NewEnumerator3
    Sprite UV Mode=::NewEnumerator3
    Uniform Sprite Size=&RandomRangeFloat(Minimum=40, Maximum=50)
    UsePositionOffset=false
    Write Material Random=true
   )
   &PartitionParticles(Number Of Partitions=4, Write Partition to Ribbon ID=::NewEnumerator0)
  ]
  particleUpdateScript [
   &ParticleState(
    Kill Particles When Lifetime Has Elapsed=false
    Let Infinitely Lived Particles Die When Emitter Deactivates=true
    Loop Particles Lifetime=true
   )
   &SocketLocation(
    SamplePosOffset=false
    Socket Index=$Particles.PartitionParticles.Partition
    Socket Reader=$System.Socket Reader
    Socket Sampling Mode=::NewEnumerator3
    VelocitySpeedLimit=false
   )
   &CameraOffset(Camera Offset Amount=10)
   &ScaleColor(
    Scale Alpha=&Lerp_Float(A=0.5, B=1, Alpha=$System.SmoothedAudioScale)
    Scale RGB=&VectorFromFloat(Value=&FN_FrontEndFloatModifier(Gameplay=0.5, Front End=0.25))
    ScaleA=true
    ScaleRGB=true
   )
   &ScaleSpriteSize(Scale Factor=&Vector2DFromFloat(Value=&Multiply_Float(
      A=&FloatFromCurve(
       FloatCurve=DataInterface() {propertyValues {Curve {keys [
           {interpMode=RCIM_Cubic, tangentMode=RCTM_User, time=0, value=1}
           {interpMode=RCIM_Cubic, tangentMode=RCTM_User, time=0.5, value=1.1}
           {tangentMode=RCTM_User, time=1, value=1}
          ]}}}
       Scale Curve=&SelectFloatFromArray(
        Direct Array Index=$Particles.PartitionParticles.Partition
        Float Selection Array=DataInterface() {propertyValues {FloatData [0.8,0.8,0.4,0.4]}}
       )
      )
      B=&Lerp_Float(A=1, B=1.5, Alpha=$System.SmoothedAudioScale)
     )))
   &FN_Sprite_Attributes(
    Color=&MultiplyLinearColor(Linear Color=$Particles.Color, Scale Factor=<1 1 1 1>)
    Create Color Attribute=true
    Create SpriteSize Attribute=true
    SpriteSize=&Multiply_Vector2D(A=$Particles.SpriteSize, B=<2 2>)
    Write to Module Namespace=true
   )
   &DynamicMaterialParameters(
    Index 0 Param 1=1.5
    Index 0 Param 3=5
    Index 0 Param 4=&RandomRangeFloat(Minimum=0, Maximum=1)
    Index 2 Param 1=3
    Index 2 Param 2=1
    Index 2 Param 3=1
    Index 2 Param 4=1
    Module.Param0WriteEnabled=true
    Module.Param0WriteEnabled2=true
    Module.Param1WriteEnabled=true
    Module.Param1WriteEnabled2=true
    Module.Param2WriteEnabled=true
    Module.Param2WriteEnabled2=true
    Module.Param3WriteEnabled=true
    Module.Param3WriteEnabled2=true
    Write Parameter Index 2=true
   )
  ]
  rendererClasses [&NiagaraSpriteRendererProperties]
  renderers [
   {rendererClass=&NiagaraSpriteRendererProperties}
   {rendererClass=&NiagaraSpriteRendererProperties, rendererIndex=1}
  ]
 }
 Ribbon {
  emitterUpdateScript [
   &EmitterState(
    Inactive Response=::NewEnumerator1
    Life Cycle Mode=::NewEnumerator1
    Loop Behavior=::NewEnumerator1
    Loop Duration Mode=::NewEnumerator1
    UseLoopDelay=true
   )
   &SpawnBurst_Instantaneous(Spawn Count=4, Use Spawn Probability=false, UseLoopCountLimit=false)
  ]
  particleSpawnScript [
   &PartitionParticles(Number Of Partitions=4, Write Partition to Ribbon ID=::NewEnumerator0)
   &InitializeParticle(
    Color=<rgba 3 1.0606 0.9844 1>
    Sprite Size=&SelectVector2DFromArray(
     Direct Array Index=$Particles.PartitionParticles.Partition
     Vector 2D Selection Array=DataInterface() {propertyValues {FloatData [<60 230>, <60 230>, <30 100>, <30 100>]}}
    )
    Sprite Size Mode=::NewEnumerator4
    UsePositionOffset=false
    Write Material Random=true
   )
   &SpriteFacingAndAlignment(Sprite Alignment=<0 0 1>, WriteAlignment=true, WriteFacing=false)
  ]
  particleUpdateScript [
   &ParticleState(Kill Particles When Lifetime Has Elapsed=false, Let Infinitely Lived Particles Die When Emitter Deactivates=true)
   &SocketLocation(
    Interpolate Socket Sampling=true
    SamplePosOffset=false
    Socket Index=$Particles.PartitionParticles.Partition
    Socket Reader=$System.Socket Reader
    Socket Sampling Mode=::NewEnumerator3
    VelocitySpeedLimit=false
   )
   &ScaleSpriteSize(
    Scale Factor=&Vector2DFromCurve(Vector2Curve=DataInterface() {propertyValues {XCurve {keys [<at 0 1>,<at 1 1>]}, YCurve {keys [<at 0 1>,<at 1 1>]}}}, ScaleCurve=<1 1>)
    Scale Sprite Size Mode=::NewEnumerator2
   )
   &DynamicMaterialParameters(
    Index 0 Param 1=&SelectFloatFromArray(
     Direct Array Index=$Particles.PartitionParticles.Partition
     Float Selection Array=DataInterface() {propertyValues {FloatData [0,0.85,0.333,0.5785]}}
    )
    Index 0 Param 2=2
    Index 0 Param 3=1
    Module.Index0Param0Write=true
    Module.Index0Param1Write=true
    Module.Index0Param2Write=true
    Module.Index0Param3Write=false
   )
   &FN_Light_Attributes(Color=<rgba 1 3.833 5 1>, Create Color Attribute=true, Create Exponent Attribute=true, Create Radius Attribute=true, Exponent=5, Radius=15)
  ]
  rendererClasses [&NiagaraSpriteRendererProperties, &NiagaraLightRendererProperties]
  renderers [
   {rendererClass=&NiagaraSpriteRendererProperties}
   {rendererClass=&NiagaraLightRendererProperties, rendererIndex=1}
  ]
 }
 Sparks {
  simTarget=GPUComputeSim
  emitterUpdateScript [
   &EmitterState(Life Cycle Mode=::NewEnumerator1, Loop Duration=1, UseLoopDelay=true)
   &SpawnRate(SpawnRate=&RandomRangeFloat(Minimum=50, Maximum=100), Use Spawn Probability=false)
  ]
  particleSpawnScript [
   /SetVariables(Particles.Acceleration:Vector3f=&RandomRangeVector(Minimum=<0 0 300>, Maximum=<0 0 700>, Evaluation Type=::NewEnumerator1)) {isSetParameters=true}
   &PartitionParticles(Number Of Partitions=4, Write Partition to Ribbon ID=::NewEnumerator0)
   &InitializeParticle(
    Color=&RandomRangeLinearColor(Minimum=<rgba 0 0.3632 1 1>, Maximum=<rgba 0 0.1963 1 1>)
    Lifetime=&RandomRangeFloat(Minimum=0.15, Maximum=0.3, Evaluation Type=::NewEnumerator1)
    Sprite Size=&Vector2DFromFloat(Value=&RandomRangeFloat(Minimum=3, Maximum=6, Evaluation Type=::NewEnumerator1))
    Sprite Size Mode=::NewEnumerator4
    UsePositionOffset=false
    Write Material Random=true
   )
   &SocketLocation(
    Interpolate Socket Sampling=true
    SamplePosOffset=true
    Sampled Position Offset=&RandomRangeVector(Minimum=<-1 -1 -1>, Maximum=<1 1 1>)
    Socket Index=$Particles.PartitionParticles.Partition
    Socket Reader=$System.Socket Reader
    Socket Sampling Mode=::NewEnumerator3
    VelocitySpeedLimit=false
   )
   &ShapeLocation(MaskSpawnGroup=false, Override Seed=false, Shape Origin=$Output.SocketLocation.SampledPosition, Sphere Radius=10, UseOwnerScale=true)
   &AddVelocity(Velocity=&FN_FrontEndVector3Modifier(
     Gameplay=&Lerp_Vector(
      A=<0 0 100>
      B=&MakeVector(
       X=&SelectFloatFromArray(
        Direct Array Index=$Particles.PartitionParticles.Partition
        Float Selection Array=DataInterface() {propertyValues {FloatData [50,-50,25,-25]}}
       )
       Y=75
       Z=15
      )
      Alpha=$System.Dive
     )
     Front End=&Lerp_Vector(A=<0 -25 100>, B=<0 -150 150>, Alpha=$System.Dive)
    ))
   &AddVelocity(
    Module.Use Falloff Distance=false
    Override Seed=false
    Velocity Mode=::NewEnumerator1
    Velocity Origin=$Particles.Position
    Velocity Speed=50
   ) {module=AddVelocity001}
  ]
  particleUpdateScript [
   &ParticleState()
   &ScaleColor(
    Scale Alpha=&FloatFromCurve(
     FloatCurve=DataInterface() {propertyValues {Curve {keys [{arriveTangent=-0.7353, interpMode=RCIM_Cubic, time=0, value=1}, {arriveTangent=-2.0072, leaveTangent=-2.0072, time=1, value=0}]}}}
     Scale Curve=&RandomRangeFloat(Minimum=0.25, Maximum=0.5, Evaluation Type=::NewEnumerator1)
    )
    Scale RGB=&VectorFromFloat(Value=5)
    ScaleA=true
    ScaleRGB=true
   )
   &DynamicMaterialParameters(
    Index 0 Param 1=0.5
    Index 0 Param 2=0.5
    Index 0 Param 3=1
    Index 0 Param 4=1
    Module.Index0Param0Write=true
    Module.Index0Param1Write=true
    Module.Index0Param2Write=true
    Module.Index0Param3Write=true
   )
   &AccelerationForce(Acceleration=$Particles.Acceleration)
   &Drag(Drag=&RandomRangeFloat(Minimum=0, Maximum=2.5), Use Linear Drag=true, Use Rotational Drag=false)
   &SolveForcesAndVelocity(Clamp Velocity=false, Limit Acceleration=false)
  ]
  rendererClasses [&NiagaraSpriteRendererProperties]
  renderers [{rendererClass=&NiagaraSpriteRendererProperties}]
 }
 Sprite_Decal_Line {
  emitterUpdateScript [
   &EmitterState()
   &SpawnRate(SpawnRate=4, Use Spawn Probability=false)
  ]
  particleSpawnScript [
   &InitializeParticle(
    Color=<rgba 0.4193 0.673 1 0.03>
    Sprite Size=<40 60>
    Sprite Size Mode=::NewEnumerator4
    UsePositionOffset=false
    Write Material Random=true
   ) {module=InitializeParticle001}
   &PartitionParticles(Number Of Partitions=4, Write Partition to Ribbon ID=::NewEnumerator0)
  ]
  particleUpdateScript [
   &ParticleState()
   &SocketLocation(
    SamplePosOffset=false
    Socket Index=$Particles.PartitionParticles.Partition
    Socket Reader=$System.Socket Reader
    Socket Sampling Mode=::NewEnumerator3
    VelocitySpeedLimit=false
   )
  ]
  rendererClasses [&NiagaraSpriteRendererProperties]
  renderers [{rendererClass=&NiagaraSpriteRendererProperties}]
 }
}
summary {
 systemName=NS_Contrail_EthicalNoggin
 emitters [
  {emitterName=Ribbon, rendererClasses [&NiagaraSpriteRendererProperties, &NiagaraLightRendererProperties]}
  {emitterName=Sparks, simTarget=GPUComputeSim, rendererClasses [&NiagaraSpriteRendererProperties]}
  {emitterName=Decal_Glow, rendererClasses [&NiagaraSpriteRendererProperties]}
  {emitterName=Sprite_Decal_Line, rendererClasses [&NiagaraSpriteRendererProperties]}
 ]
 userVariables [
  User.bIsFrontEnd {type=NiagaraBool}
  User.bIsFrontEndPreview {type=NiagaraBool}
  User.ParaGlideLeanAlpha {type=NiagaraFloat, defaultValue=0, description=""}
  User.SkeletalMesh {
   type=NiagaraDataInterfaceSkeletalMesh
   defaultValue {
    struct=NiagaraExt_VariableValue_DataInterface
    value {dataInterface=&self:NiagaraDataInterfaceSkeletalMesh_0}
   }
  }
 ]
}
